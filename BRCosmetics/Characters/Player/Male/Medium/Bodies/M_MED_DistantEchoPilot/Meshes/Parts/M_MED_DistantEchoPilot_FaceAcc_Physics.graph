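# Physics asset for the face accessory part of the male medium DistantEchoPilot character, giving its dynamic hair bones physical bodies.
# Small capsules sit on each hair bone, and two capsules represent the head.
# Angular-limited constraints root five strands to the head: a three-bone right chain, two-bone left and right side chains, and two single-bone strands.
# Tight twist and mostly narrow, asymmetric swing limits restrict how far the hair can move.
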PhysicsAsset /BRCosmetics/Characters/Player/Male/Medium/Bodies/M_MED_DistantEchoPilot/Meshes/Parts/M_MED_DistantEchoPilot_FaceAcc_Physics
bodies {
 dyn_hair_a_l_01 [{shapeType=Capsule, length=1, radius=0.75, shapeName=dyn_hair_a_l_01_capsule, center=<0.9008 0 1.5089>, rotation=<rot 44.0141 -180 -180>}]
 dyn_hair_a_r_01 [{shapeType=Capsule, length=3, radius=1, shapeName=dyn_hair_a_r_01_capsule, center=<0.9788 0 2.0756>, rotation=<rot 27.3649 180 180>}]
 dyn_hair_a_r_02 [{shapeType=Capsule, length=2, radius=1, shapeName=dyn_hair_a_r_02_capsule, center=<1 0 0>, rotation=<rot 90 0 0>}]
 dyn_hair_a_r_03 [{shapeType=Capsule, length=2, radius=1, shapeName=dyn_hair_a_r_03_capsule, center=<1 0 0>, rotation=<rot 76.1415 0 0>}]
 dyn_hair_b_r_01 [{shapeType=Capsule, length=1, radius=0.75, shapeName=dyn_hair_a_l_01_capsule, center=<0.9008 0 1.5089>, rotation=<rot 44.0141 -180 -180>}]
 dyn_hair_sl_01 [{shapeType=Capsule, length=2, radius=1, shapeName=dyn_hair_sl_01_capsule, center=<1 0 0>, rotation=<rot 90 0 0>}]
 dyn_hair_sl_02 [{shapeType=Capsule, length=2, radius=1, shapeName=dyn_hair_sl_02_capsule, center=<1 0 0>, rotation=<rot 90 0 0>}]
 dyn_hair_sr_01 [{shapeType=Capsule, length=2, radius=1, shapeName=dyn_hair_sr_01_capsule, center=<1 0 0>, rotation=<rot 90 0 0>}]
 dyn_hair_sr_02 [{shapeType=Capsule, length=2, radius=1, shapeName=dyn_hair_sr_02_capsule, center=<1 0 0>, rotation=<rot 90 0 0>}]
 head [
  {shapeType=Capsule, length=4.25, radius=5.5, center=<4.2022 5.0833 0>, rotation=<rot 0 0 90>}
  {shapeType=Capsule, length=1, radius=1, shapeName=head_capsule, center=<6.7523 5.2872 5.8614>, rotation=<rot 90 -166.7373 -166.7373>}
 ]
}
constraints [
 {bone1Name=dyn_hair_b_r_01, bone2Name=head, swing1LimitDegrees=5, swing2LimitDegrees=40, twistLimitDegrees=2}
 {bone1Name=dyn_hair_a_r_03, bone2Name=dyn_hair_a_r_02, swing1LimitDegrees=25, swing2LimitDegrees=25, twistLimitDegrees=2}
 {bone1Name=dyn_hair_a_r_02, bone2Name=dyn_hair_a_r_01, swing1LimitDegrees=5, swing2LimitDegrees=10, twistLimitDegrees=2}
 {bone1Name=dyn_hair_a_r_01, bone2Name=head, swing1LimitDegrees=2, swing2LimitDegrees=10, twistLimitDegrees=2}
 {bone1Name=dyn_hair_sr_01, bone2Name=head, swing1LimitDegrees=20, swing2LimitDegrees=2, twistLimitDegrees=2}
 {bone1Name=dyn_hair_sr_02, bone2Name=dyn_hair_sr_01, swing1LimitDegrees=10, swing2LimitDegrees=2, twistLimitDegrees=2}
 {bone1Name=dyn_hair_sl_02, bone2Name=dyn_hair_sl_01, swing1LimitDegrees=10, swing2LimitDegrees=2, twistLimitDegrees=2}
 {bone1Name=dyn_hair_sl_01, bone2Name=head, swing1LimitDegrees=20, swing2LimitDegrees=2, twistLimitDegrees=2}
 {bone1Name=dyn_hair_a_l_01, bone2Name=head, swing1LimitDegrees=5, swing2LimitDegrees=40, twistLimitDegrees=2}
]
