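# Animation Blueprint for the DistantEchoPilot face accessory mesh that adds secondary hair motion on top of the attached parent's pose.
# Copy Pose From Mesh feeds a serial chain of partially blended Anim Dynamics nodes for right, left, side and back hair bones.
# Custom gravity and angular limits shape the motion, with springs on some bones and a head-driven spherical limit on one.
# The RigidBody node is unconnected and the event graph holds no logic.
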AnimBlueprint /BRCosmetics/Characters/Player/Male/Medium/Bodies/M_MED_DistantEchoPilot/Meshes/Parts/M_MED_DistantEchoPilot_FaceAcc_AnimBP
target_skeleton=&/M_MED_DistantEchoPilot_FaceAcc_Skeleton
graphs [
 AnimGraph {nodes [
   RigidBody!(ComponentPose, Alpha=0, OverrideWorldGravity:Vector="0,0,-980", ExternalForce) -> (Pose) {
    bOverrideWorldGravity=true
    overridePhysicsAsset=&/M_MED_DistantEchoPilot_FaceAcc_Physics
    simulationSpace=BaseBoneSpace
    alphaBoolBlend {blendOption=Linear}
    baseBoneRef {boneName=head}
    componentAppliedLinearAccClamp=<3000 3000 3000>
    componentLinearAccScale=<0.5 0.5 0.9>
    componentLinearVelScale=<0.5 0.5 0.9>
    overrideWorldGravity {z=-980}
   }
   n1 = Anim Dynamics - Bone: dyn_hair_a_r_01(ComponentPose=Local To Component(LocalPose=Copy Pose From Mesh() {bUseAttachedParent=true}), Alpha=0.4, ExternalForce="0,0,0") {
    angularDampingOverride=0.999
    angularSpringConstant=30000
    bAngularSpring=true
    bGravityOverrideInSimSpace=true
    bOverrideAngularDamping=true
    bUseGravityOverride=true
    alphaBoolBlend {blendOption=Linear}
    boundBone {boneName=dyn_hair_a_r_01}
    componentAppliedLinearAccClamp=<2000 2000 2000>
    componentLinearAccScale=<0.1 0.1 0.2>
    componentLinearVelScale=<0.1 0.1 0.4>
    gravityOverride {z=-400}
    physicsBodyDefinitions [{
      boundBone {boneName=dyn_hair_a_r_01}
      boxExtents=<20 20 30>
      constraintSetup {angularLimitsMax=<0.01 0 1>, angularLimitsMin=<-0.01 -10 -1>, angularTarget=<0 0 0>}
      localJointOffset=<4 0 0>
     }]
    relativeSpaceBone {boneName=head}
   }
   n2 = Anim Dynamics - Bone: dyn_hair_a_r_02(ComponentPose=@n1, Alpha=0.7, ExternalForce="0,0,0") {
    angularDampingOverride=0.9
    angularSpringConstant=3000
    bAngularSpring=true
    bGravityOverrideInSimSpace=true
    bOverrideAngularDamping=true
    bUseGravityOverride=true
    bUsePlanarLimit=false
    alphaBoolBlend {blendOption=Linear}
    boundBone {boneName=dyn_hair_a_r_02}
    componentAppliedLinearAccClamp=<2000 2000 2000>
    componentLinearAccScale=<0.3 0.3 0.5>
    componentLinearVelScale=<0.3 0.3 0.5>
    gravityOverride {z=-400}
    physicsBodyDefinitions [{
      boundBone {boneName=dyn_hair_a_r_02}
      boxExtents=<20 20 30>
      constraintSetup {angularLimitsMax=<10 10 5>, angularLimitsMin=<-10 -10 -15>, angularTarget=<0 0 0>}
      localJointOffset=<4 0 0>
     }]
    relativeSpaceBone {boneName=head}
   }
   n3 = Anim Dynamics - Bone: dyn_hair_a_r_03(ComponentPose=@n2, Alpha=0.7, ExternalForce="0,0,0") {
    angularDampingOverride=0.9
    angularSpringConstant=3000
    bAngularSpring=true
    bGravityOverrideInSimSpace=true
    bOverrideAngularDamping=true
    bUseGravityOverride=true
    bUsePlanarLimit=false
    bUseSphericalLimits=true
    alphaBoolBlend {blendOption=Linear}
    boundBone {boneName=dyn_hair_a_r_03}
    componentAppliedLinearAccClamp=<2000 2000 2000>
    componentLinearAccScale=<0.35 0.5 0.5>
    componentLinearVelScale=<0.35 0.35 0.5>
    gravityOverride {z=-400}
    physicsBodyDefinitions [{
      boundBone {boneName=dyn_hair_a_r_03}
      boxExtents=<20 20 30>
      constraintSetup {angularLimitsMax=<15 15 5>, angularLimitsMin=<-15 -15 -45>, angularTarget=<0 0 0>}
      localJointOffset=<4 0 0>
     }]
    relativeSpaceBone {boneName=head}
    sphericalLimits [{limitRadius=10, drivingBone {boneName=head}, sphereLocalOffset=<7 2 0>}]
   }
   n4 = Anim Dynamics - Bone: dyn_hair_a_l_01(ComponentPose=@n3, Alpha=0.7, ExternalForce="0,0,0") {
    angularDampingOverride=0.999
    angularSpringConstant=500
    bAngularSpring=true
    bGravityOverrideInSimSpace=true
    bOverrideAngularDamping=true
    bUseGravityOverride=true
    simulationSpace=RootRelative
    alphaBoolBlend {blendOption=Linear}
    boundBone {boneName=dyn_hair_a_l_01}
    componentAppliedLinearAccClamp=<2000 2000 2000>
    componentLinearAccScale=<0.2 0.2 0.4>
    componentLinearVelScale=<0.2 0.2 0.4>
    gravityOverride {z=-400}
    physicsBodyDefinitions [{
      boundBone {boneName=dyn_hair_a_l_01}
      boxExtents=<20 20 30>
      constraintSetup {angularLimitsMax=<0.01 0 1>, angularLimitsMin=<-0.01 -10 -1>, angularTarget=<0 -10 0>}
      localJointOffset=<4 0 0>
     }]
    relativeSpaceBone {boneName=head}
   }
   n5 = Anim Dynamics - Bone: dyn_hair_b_r_01(ComponentPose=@n4, Alpha=0.7, ExternalForce="0,0,0") {
    angularDampingOverride=0.999
    angularSpringConstant=500
    bAngularSpring=true
    bGravityOverrideInSimSpace=true
    bOverrideAngularDamping=true
    bUseGravityOverride=true
    simulationSpace=RootRelative
    alphaBoolBlend {blendOption=Linear}
    boundBone {boneName=dyn_hair_b_r_01}
    componentAppliedLinearAccClamp=<2000 2000 2000>
    componentLinearAccScale=<0.2 0.2 0.4>
    componentLinearVelScale=<0.2 0.2 0.4>
    gravityOverride {z=-400}
    physicsBodyDefinitions [{
      boundBone {boneName=dyn_hair_b_r_01}
      boxExtents=<20 20 30>
      constraintSetup {angularLimitsMax=<0.01 0 1>, angularLimitsMin=<-0.01 -10 -1>, angularTarget=<0 -10 0>}
      localJointOffset=<4 0 0>
     }]
    relativeSpaceBone {boneName=head}
   }
   n6 = Anim Dynamics - Bone: dyn_hair_sl_01(ComponentPose=@n5, Alpha=0.6, ExternalForce="0,0,0") {
    angularDampingOverride=0.999
    angularSpringConstant=1000
    bAngularSpring=true
    bGravityOverrideInSimSpace=true
    bOverrideAngularDamping=true
    bUseGravityOverride=true
    simulationSpace=RootRelative
    alphaBoolBlend {blendOption=Linear}
    boundBone {boneName=dyn_hair_sl_01}
    componentAppliedLinearAccClamp=<2000 2000 2000>
    componentLinearAccScale=<0.1 0.1 0.5>
    componentLinearVelScale=<0.1 0.1 0.5>
    gravityOverride {z=-400}
    physicsBodyDefinitions [{
      boundBone {boneName=dyn_hair_sl_01}
      boxExtents=<20 20 30>
      constraintSetup {angularLimitsMax=<0.01 2 5>, angularLimitsMin=<-0.01 -2 -5>, angularTarget=<0 -10 0>}
      localJointOffset=<3 0 0>
     }]
    relativeSpaceBone {boneName=head}
   }
   n7 = Anim Dynamics - Bone: dyn_hair_sl_02(ComponentPose=@n6, Alpha=0.6, ExternalForce="0,0,0") {
    angularDampingOverride=0.9
    angularSpringConstant=300
    bGravityOverrideInSimSpace=true
    bOverrideAngularDamping=true
    bUseGravityOverride=true
    bUsePlanarLimit=false
    alphaBoolBlend {blendOption=Linear}
    boundBone {boneName=dyn_hair_sl_02}
    componentAppliedLinearAccClamp=<2000 2000 2000>
    componentLinearAccScale=<0.3 0.3 0.8>
    componentLinearVelScale=<0.3 0.3 0.8>
    gravityOverride {z=-400}
    physicsBodyDefinitions [{
      boundBone {boneName=dyn_hair_sl_02}
      boxExtents=<20 20 30>
      constraintSetup {angularLimitsMax=<0 2 5>, angularLimitsMin=<0 -2 -5>, angularTarget=<0 -10 0>}
      localJointOffset=<2 0 0>
     }]
    relativeSpaceBone {boneName=head}
   }
   n8 = Anim Dynamics - Bone: dyn_hair_sr_01(ComponentPose=@n7, Alpha=0.6, ExternalForce="0,0,0") {
    angularDampingOverride=0.999
    angularSpringConstant=1000
    bAngularSpring=true
    bGravityOverrideInSimSpace=true
    bOverrideAngularDamping=true
    bUseGravityOverride=true
    simulationSpace=RootRelative
    alphaBoolBlend {blendOption=Linear}
    boundBone {boneName=dyn_hair_sr_01}
    componentAppliedLinearAccClamp=<2000 2000 2000>
    componentLinearAccScale=<0.1 0.1 0.5>
    componentLinearVelScale=<0.1 0.1 0.5>
    gravityOverride {z=-400}
    physicsBodyDefinitions [{
      boundBone {boneName=dyn_hair_sr_01}
      boxExtents=<20 20 30>
      constraintSetup {angularLimitsMax=<0.01 2 5>, angularLimitsMin=<-0.01 -2 -5>, angularTarget=<0 -10 0>}
      localJointOffset=<3 0 0>
     }]
    relativeSpaceBone {boneName=head}
   }
   n9 = Anim Dynamics - Bone: dyn_hair_sr_02(ComponentPose=@n8, Alpha=0.6, ExternalForce="0,0,0") {
    angularDampingOverride=0.9
    angularSpringConstant=300
    bGravityOverrideInSimSpace=true
    bOverrideAngularDamping=true
    bUseGravityOverride=true
    bUsePlanarLimit=false
    alphaBoolBlend {blendOption=Linear}
    boundBone {boneName=dyn_hair_sr_02}
    componentAppliedLinearAccClamp=<2000 2000 2000>
    componentLinearAccScale=<0.3 0.3 0.8>
    componentLinearVelScale=<0.3 0.3 0.8>
    gravityOverride {z=-400}
    physicsBodyDefinitions [{
      boundBone {boneName=dyn_hair_sr_02}
      boxExtents=<20 20 30>
      constraintSetup {angularLimitsMax=<0 2 5>, angularLimitsMin=<0 -2 -5>, angularTarget=<0 -10 0>}
      localJointOffset=<2 0 0>
     }]
    relativeSpaceBone {boneName=head}
   }
   n10 = Anim Dynamics - Bone: dyn_hair_a_bk_01(ComponentPose=@n9, Alpha=0.6, ExternalForce="0,0,0") {
    angularDampingOverride=0.999
    angularSpringConstant=500
    bGravityOverrideInSimSpace=true
    bUseGravityOverride=true
    bUsePlanarLimit=false
    simulationSpace=RootRelative
    alphaBoolBlend {blendOption=Linear}
    boundBone {boneName=dyn_hair_a_bk_01}
    componentLinearAccScale=<1 1 5>
    componentLinearVelScale=<1 1 5>
    gravityOverride {z=-400}
    physicsBodyDefinitions [{
      boundBone {boneName=dyn_hair_a_bk_01}
      boxExtents=<10 10 20>
      constraintSetup {angularLimitsMax=<0 0 10>, angularLimitsMin=<0 -10 -10>, angularTarget=<0 0 0>}
      localJointOffset=<5 0 0>
     }]
    relativeSpaceBone {boneName=head}
   }
   n11 = Anim Dynamics - Bone: dyn_hair_a_bk_02(ComponentPose=@n10, Alpha=0.6, ExternalForce="0,0,0") {
    angularDampingOverride=0.999
    angularSpringConstant=500
    bGravityOverrideInSimSpace=true
    bUseGravityOverride=true
    bUsePlanarLimit=false
    simulationSpace=RootRelative
    alphaBoolBlend {blendOption=Linear}
    boundBone {boneName=dyn_hair_a_bk_02}
    componentLinearAccScale=<1 1 5>
    componentLinearVelScale=<1 1 5>
    gravityOverride {z=-400}
    physicsBodyDefinitions [{
      boundBone {boneName=dyn_hair_a_bk_02}
      boxExtents=<10 10 20>
      constraintSetup {angularLimitsMax=<0 0 15>, angularLimitsMin=<0 -10 -15>, angularTarget=<0 0 0>}
      localJointOffset=<1 0 0>
     }]
    relativeSpaceBone {boneName=head}
   }
   n12 = Anim Dynamics - Bone: dyn_hair_b_bk_01(ComponentPose=@n11, Alpha=0.6, ExternalForce="0,0,0") {
    angularDampingOverride=0.999
    angularSpringConstant=500
    bGravityOverrideInSimSpace=true
    bUseGravityOverride=true
    bUsePlanarLimit=false
    simulationSpace=RootRelative
    alphaBoolBlend {blendOption=Linear}
    boundBone {boneName=dyn_hair_b_bk_01}
    componentLinearAccScale=<1 1 5>
    componentLinearVelScale=<1 1 5>
    gravityOverride {z=-400}
    physicsBodyDefinitions [{
      boundBone {boneName=dyn_hair_b_bk_01}
      boxExtents=<10 10 20>
      constraintSetup {angularLimitsMax=<0 0 10>, angularLimitsMin=<0 -10 -10>, angularTarget=<0 0 0>}
      localJointOffset=<5 0 0>
     }]
    relativeSpaceBone {boneName=head}
   }
   n13 = Anim Dynamics - Bone: dyn_hair_b_bk_02(ComponentPose=@n12, Alpha=0.6, ExternalForce="0,0,0") {
    angularDampingOverride=0.999
    angularSpringConstant=500
    bGravityOverrideInSimSpace=true
    bUseGravityOverride=true
    bUsePlanarLimit=false
    simulationSpace=RootRelative
    alphaBoolBlend {blendOption=Linear}
    boundBone {boneName=dyn_hair_b_bk_02}
    componentLinearAccScale=<1 1 5>
    componentLinearVelScale=<1 1 5>
    gravityOverride {z=-400}
    physicsBodyDefinitions [{
      boundBone {boneName=dyn_hair_b_bk_02}
      boxExtents=<10 10 20>
      constraintSetup {angularLimitsMax=<0 0 15>, angularLimitsMin=<0 -10 -15>, angularTarget=<0 0 0>}
      localJointOffset=<1 0 0>
     }]
    relativeSpaceBone {boneName=head}
   }
   Output Pose(Result=Component To Local(ComponentPose=@n13)) {layerGroup=None}
  ]}
 EventGraph {nodes [Event BlueprintUpdateAnimation()]}
]
variables []
